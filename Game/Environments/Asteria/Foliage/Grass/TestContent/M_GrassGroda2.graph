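# Two-sided subsurface grass material for instanced meshes in the Asteria environment, with wind sway and color variation.
# Panned wind-texture noise and MPC wind direction bend blades around UV1-stored pivots, by axis rotation or a switchable offset-normalize path.
# World-position cloud noise adds height variation.
# Vertex color masks bend, height, hue and value, while world-space normals are mirrored across UV halves, flattened, pushed upward and adjusted for bending.
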Material /Game/Environments/Asteria/Foliage/Grass/TestContent/M_GrassGroda2
graph {
 properties {
  bTangentSpaceNormal=false
  bUsedWithInstancedStaticMeshes=true
  pixelDepthOffsetMode=PDOM_Legacy
  shadingModel=MSM_Subsurface
  twoSided=true
  usedShadingModels=Subsurface
 }
 nodes [
  n1 = NamedRerouteUsage()
  n2 = NamedRerouteUsage()
  n3 = Distance(A=NamedRerouteUsage(), B=NamedRerouteUsage())
  n4 = Multiply(
   A=ScalarParameter() {defaultValue=2.2548, parameterName=OffsetAmount}
   B=Power(Base=VertexColor().B, Exp=ScalarParameter() {defaultValue=1.5343, parameterName=Falloff})
  )
  n5 = Multiply(A=Multiply(A=Normalize(VectorInput=@n2)) {constB=-1}, B=Multiply(A=NamedRerouteUsage(), B=Multiply(A=@n3, B=@n4)))
  n6 = Normalize(VectorInput=Subtract(A=Add(A=@n1, B=@n5), B=NamedRerouteUsage()))
  n7 = VectorParameter() {parameterName=Normal, defaultValue=<rgba 0.6741 0 1 1>}
  n8 = Constant3Vector() {constant=<rgba -1 1 1 0>}
  n9 = TextureCoordinate()
  n10 = Ceil(None=Subtract(A=ComponentMask(None=@n9) {r=true}) {constB=0.5})
  n11 = MaterialFunctionCall(
   Normal=LinearInterpolate(A=@n7.RGB, B=Multiply(A=@n7.RGB, B=@n8), Alpha=Saturate(None=@n10))
   Flatness=ScalarParameter() {defaultValue=0.8272, parameterName=Normal}
  ) {materialFunction=&FlattenNormal}
  n12 = Normalize(VectorInput=LinearInterpolate(
    A=Transform(None=@n11.Result)
    B=Constant3Vector() {constant=<rgba 0 0 1 0>}
    Alpha=ScalarParameter() {defaultValue=0.2508, parameterName=ForceUP}
   ))
  Add(A=@n6, B=@n12)
  n13 = WorldPosition()
  n14 = TextureSample(UVs=Divide(A=ComponentMask(None=@n13.XYZ) {g=true, r=true}, B=ScalarParameter() {defaultValue=5610.6743, parameterName=SizeMask})) {texture=&AtmosphericCloudNoise01}
  n15 = ScalarParameter() {defaultValue=1.782, parameterName=C}
  n16 = VertexColor()
  n17 = Multiply(
   A=Multiply(A=Power(Base=@n14.R, Exp=@n15), B=Multiply(A=@n16.G, B=@n16.G))
   B=AppendVector(A=Constant2Vector(), B=ScalarParameter() {defaultValue=201.6037, parameterName=HeightVar})
  )
  n18 = ComponentMask(None=CollectionParameter() {collection=/Game/Environments/Asteria/Foliage/Wind/MPC_Wind, parameterName=WindDirection}) {b=true, g=true, r=true}
  n19 = Constant3Vector() {constant=<rgba 1 0 0 0>}
  n20 = StaticSwitchParameter(True=LinearInterpolate(A=@n18, B=@n19, Alpha=ScalarParameter() {defaultValue=0.9934, parameterName=WindDirectionInterp}), False=@n18) {parameterName=UseInterpolatingWIndDirection}
  n21 = Constant3Vector() {constant=<rgba 0 0 1 0>}
  n22 = Multiply(A=Subtract(A=NamedRerouteUsage()) {constB=0.5}, B=ScalarParameter() {defaultValue=2, parameterName=WindNoise})
  n23 = Add(A=@n20, B=Multiply(A=CrossProduct(A=@n20, B=@n21), B=@n22))
  n24 = Normalize(VectorInput=CrossProduct(A=@n23, B=Constant3Vector() {constant=<rgba 0 0 1 0>}))
  n25 = VertexColor()
  n26 = ScalarParameter() {defaultValue=0.3333, parameterName=WPOfalloff}
  n27 = Multiply(A=NamedRerouteUsage(), B=Multiply(A=ScalarParameter() {defaultValue=0.2403, parameterName=RotAngle}, B=Power(Base=@n25.B, Exp=@n26)))
  n28 = NamedRerouteUsage()
  n29 = NamedRerouteUsage()
  n30 = Add(A=Multiply(A=@n3, B=@n6), B=NamedRerouteUsage())
  n31 = NamedRerouteUsage()
  Add(
   A=@n17
   B=StaticSwitchParameter(True=RotateAboutAxis(NormalizedRotationAxis=@n24, RotationAngle=@n27, PivotPoint=@n28, Position=@n29), False=Subtract(A=@n30, B=@n31)) {defaultValue=true, parameterName=UseRotateaboutAxis}
  )
  AppendVector(A=@n22, B=Constant())
  ComponentMask(None=TextureCoordinate() {coordinateIndex=2}) {r=true}
  Constant()
  LinearInterpolate()
  n32 = TextureCoordinate()
  n33 = AppendVector(A=ScalarParameter() {defaultValue=0.9752, parameterName=test2}, B=ScalarParameter() {defaultValue=3.4016, parameterName=test3})
  MaterialFunctionCall!(
   Vector3
   MaximumNumberOfDigits
   UVs=Multiply(A=Multiply(A=@n32, B=@n33), B=ScalarParameter() {defaultValue=0.9381, parameterName=test})
   DebugTextLocation RG_UpperRight BA_LowerLeft
   Component Spacing
  ) {materialFunction=/Engine/Functions/Engine_MaterialFunctions02/Utility/DebugFloat3Values}
  n34 = VertexColor()
  n35 = LinearInterpolate(A=ScalarParameter() {parameterName=HueBottom}, B=ScalarParameter() {parameterName=HueTop}, Alpha=@n34.B)
  n36 = Multiply(
   A=Power(Base=@n34.G, Exp=ScalarParameter() {defaultValue=4.0629, parameterName=HueContrast})
   B=ScalarParameter() {defaultValue=-0.0832, parameterName=HueVar}
  )
  n37 = LinearInterpolate(
   A=VectorParameter() {parameterName=Color, defaultValue=<rgba 0.0458 0.0833 0.0055 1>}.RGB
   B=VectorParameter() {parameterName=Color2, defaultValue=<rgba 0.0458 0.0833 0.0055 1>}.RGB
   Alpha=VertexColor().B
  )
  n38 = LinearInterpolate(A=ScalarParameter() {defaultValue=1, parameterName=ValueVar}, Alpha=@n34.G)
  Multiply(
   A=MaterialFunctionCall(Hue Shift Percentage=Add(A=@n35, B=@n36), Texture=Multiply(A=@n37, B=@n38)) {materialFunction=&HueShift}.Result
   B=ScalarParameter() {defaultValue=1, parameterName=SSSColor}
  )
  n39 = NamedRerouteUsage()
  n40 = ComponentMask(None=@n39) {b=true}
  n41 = Divide(
   A=Add(A=ComponentMask(None=@n39) {g=true, r=true}, B=AppendVector(A=@n40, B=@n40))
   B=ScalarParameter() {defaultValue=3220.4573, parameterName=WindMaskScale, sliderMin=3}
  )
  n42 = ScalarParameter() {defaultValue=0.1, parameterName=WindSpeed}
  n43 = ComponentMask(None=@n18) {g=true, r=true}
  n44 = TextureSample(UVs=Panner(Coordinate=@n41, Speed=Multiply(A=@n42, B=@n43)) {bFractionalPart=true}) {samplerType=SAMPLERTYPE_LinearColor, texture=/Game/Environments/Asteria/Foliage/Wind/RT_WindGlobal_STATIC_NEW}
  n45 = ComponentMask(None=@n19) {g=true, r=true}
  n46 = TextureSample(UVs=Panner(Coordinate=@n41, Speed=Multiply(A=@n42, B=@n45)) {bFractionalPart=true}) {samplerType=SAMPLERTYPE_LinearColor, texture=&RT_WindGlobal_STATIC_NEW}
  n47 = ScalarParameter() {defaultValue=0.9934, parameterName=WindDirectionInterp}
  NamedRerouteDeclaration(None=StaticSwitchParameter(True=LinearInterpolate(A=@n44.R, B=@n46.R, Alpha=@n47), False=@n44.R) {parameterName=UseInterpolatingWIndDirection}) {name=WindIntensityNoise, nodeColor=<rgba 0.1719 0 1 1>}
  NamedRerouteDeclaration(None=Add(A=@n17, B=WorldPosition() {worldPositionShaderOffset=WPT_ExcludeAllShaderOffsets}.XYZ)) {name=WPOwithHeightOffset, nodeColor=<rgba 0.3359 0 1 1>}
  n48 = TextureCoordinate() {coordinateIndex=1}
  n49 = TransformPosition(None=MaterialFunctionCall(X=ComponentMask(None=@n48) {r=true}, Y=ComponentMask(None=@n48) {g=true}, Z=Constant()) {materialFunction=&MakeFloat3}.Result) {transformSourceType=TRANSFORMPOSSOURCE_Instance, transformType=TRANSFORMPOSSOURCE_World}
  NamedRerouteDeclaration(None=@n49) {name=PivotPosition, nodeColor=<rgba 0 0.9766 1 1>}
  NamedRerouteDeclaration(None=@n23) {name=WindOffsetVector, nodeColor=<rgba 0 1 0.4375 1>}
  n50 = RotateAboutAxis(NormalizedRotationAxis=@n24, RotationAngle=@n27, PivotPoint=Constant3Vector(), Position=@n12)
  Normalize(VectorInput=StaticSwitchParameter(
    True=Add(A=@n12, B=@n50)
    False=MaterialFunctionCall(BaseNormal=@n6, AdditionalNormal=@n12) {materialFunction=&BlendAngleCorrectedNormals}.Result
   ) {defaultValue=true, parameterName=UseRotateaboutAxis})
  Reroute(None=@n49)
  Saturate(None=LinearInterpolate(
    A=ScalarParameter() {defaultValue=1, parameterName=SSSOpacity}
    B=ScalarParameter() {defaultValue=0.5, parameterName=SSSOpacity2}
    Alpha=@n16.B
   ))
  ScalarParameter() {defaultValue=0.6684, parameterName=R}
  ScalarParameter() {defaultValue=0.5, parameterName=S}
  WorldPosition()
 ]
}
